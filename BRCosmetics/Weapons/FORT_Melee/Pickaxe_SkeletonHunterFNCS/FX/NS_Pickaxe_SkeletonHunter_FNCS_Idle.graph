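# Idle Niagara effect for the FNCS Skeleton Hunter pickaxe, emitting short-lived, bright green mesh particles.
# One emitter continuously spawns thin meshes on an offset cylinder surface with slight drift and random orientation, and samples the system skeletal mesh for location each update.
# Over each particle's life the mesh stretches sharply while alpha fades in and out.
# Curves and random values drive four dynamic material parameters.
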NiagaraSystem /BRCosmetics/Weapons/FORT_Melee/Pickaxe_SkeletonHunterFNCS/FX/NS_Pickaxe_SkeletonHunter_FNCS_Idle
dependencies {
 usedDataInterfaces [&NiagaraDataInterfaceCurve, &NiagaraDataInterfaceSkeletalMesh]
 usedDynamicInputs []
 usedModules [
  &self:SystemSpawnScript.SystemScriptSource.SystemScriptGraph.NiagaraNodeAssignment_0.SetVariables_#1
  &SystemState
 ]
}
emitters {Empty {
  emitterUpdateScript [
   &EmitterState()
   &SpawnRate(SpawnRate=6, Use Spawn Probability=false)
  ]
  particleSpawnScript [
   &InitializeParticle(
    Color=<rgba 0 1 0 60>
    Lifetime Max=1.1
    Lifetime Min=0.7
    Lifetime Mode=::NewEnumerator1
    Mesh Scale Max=<0.08 0.01 0.1>
    Mesh Scale Min=<0.16 0.015 0.1>
    Mesh Scale Mode=::NewEnumerator2
    UsePositionOffset=false
    Write Material Random=true
   )
   &CameraOffset(Camera Offset Amount=10)
   &ShapeLocation(
    Cylinder Height=40
    Cylinder Radius=4
    MaskSpawnGroup=false
    Offset=<0 20 4>
    Offset Mode=::NewEnumerator0
    Override Seed=false
    Rotation Axis=<0 1 0>
    Rotation Mode=::NewEnumerator1
    Shape Primitive=::NewEnumerator1
    Surface Only=true
    UseOwnerScale=true
   )
   &AddVelocity(Velocity=&RandomRangeVector(Minimum=<0 -3 0>, Maximum=<0 -1 0>))
   &InitialMeshOrientation(
    Mesh Orientation Mode=::NewEnumerator1
    Rotation=&RandomRangeVector(Minimum=<-1 0 0>, Maximum=<1 0 0>)
    Use Rotation Vector=true
   )
  ]
  particleUpdateScript [
   &ParticleState()
   &SkeletalMeshLocation(
    MaskSpawnGroup=false
    Mesh Sampling Type=::NewEnumerator5
    Orientation Sampling=::NewEnumerator1
    Skeletal Mesh=$System.Skeletal Mesh
    Socket Sampling Mode=::NewEnumerator5
   )
   &ScaleMeshSize(Scale Factor=&Lerp_Vector(A=<0.1 1 1>, B=<10 8 1>, Alpha=$Particles.NormalizedAge))
   &ScaleColor(
    Scale Alpha=&FloatFromCurve(FloatCurve=DataInterface() {propertyValues {Curve {keys [
         {arriveTangent=3.1835, interpMode=RCIM_Cubic, leaveTangent=3.1835, tangentMode=RCTM_User, time=0, value=0}
         {arriveTangent=3.0214, interpMode=RCIM_Cubic, leaveTangent=3.0214, time=0.2459, value=1}
         {arriveTangent=-0.6384, interpMode=RCIM_Cubic, leaveTangent=-0.6384, time=0.3971, value=1.1998}
         {arriveTangent=-2.5872, interpMode=RCIM_Cubic, leaveTangent=-2.5872, time=0.6259, value=0.7574}
         {arriveTangent=-2.1791, interpMode=RCIM_Cubic, leaveTangent=-2.1791, time=0.7543, value=0.2756}
         {arriveTangent=-1.1217, interpMode=RCIM_Cubic, leaveTangent=-1.1217, time=0.9211, value=0.1142}
         <at 1 0 RCIM_Cubic>
        ]}}})
    Scale RGB=&VectorFromFloat(Value=5)
    ScaleA=true
    ScaleRGB=true
   )
   &SolveForcesAndVelocity(Clamp Velocity=false, Limit Acceleration=false)
   &DynamicMaterialParameters(
    Index 0 Param 1=&FloatFromCurve(FloatCurve=DataInterface() {propertyValues {Curve {keys [<at 0 1>,<at 1 0>]}}}, Scale Curve=5)
    Index 0 Param 2=&RandomRangeFloat(Minimum=30, Maximum=50)
    Index 0 Param 3=&FloatFromCurve(FloatCurve=DataInterface() {propertyValues {Curve {keys [<at 0 3>,<at 1 12>]}}})
    Index 0 Param 4=&FloatFromCurve(FloatCurve=DataInterface() {propertyValues {Curve {keys [
         <at 0.0123 0.25 RCIM_Cubic>
         {arriveTangent=-0.1667, interpMode=RCIM_Cubic, leaveTangent=-0.1667, time=0.8212, value=0.2237}
         {arriveTangent=-1.251, interpMode=RCIM_Cubic, leaveTangent=-1.251, time=0.9123, value=0.1}
         <at 1 0 RCIM_Cubic>
        ]}}})
    Module.Index0Param0Write=true
    Module.Index0Param1Write=true
    Module.Index0Param2Write=true
    Module.Index0Param3Write=true
   )
  ]
  rendererClasses [&NiagaraMeshRendererProperties]
  renderers [{rendererClass=&NiagaraMeshRendererProperties, rendererIndex=0}]
 }}
summary {
 systemName=NS_Pickaxe_SkeletonHunter_FNCS_Idle
 emitters [{emitterName=Empty, rendererClasses [&NiagaraMeshRendererProperties]}]
 userVariables []
}
